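# Looping Sound Cue for a jungle wall sconce fire, played as a quiet, low-priority positional ambient.
# One of three random torch loops, mixed at half level, layers over a stereo coliseum brazier loop, with slight random pitch modulation.
# Nested envelopers add a short fade-in and endlessly looping, slow volume and pitch swells on differing periods to vary the loop.
# A shared attenuation asset handles falloff, and a concurrency override caps simultaneous voices.
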SoundCue /Game/Sounds/Asteria/Environment_Actors/Jungle_WallSconceFire/Spot_WallSconce_Fire_Loop_02_Cue
root=Enveloper_1
nodes [
 {
  id=Enveloper_0
  children [Modulator_1]
  properties {
   bLoop=true
   bLoopIndefinitely=true
   loopEnd=30
   pitchCurve {editorCurveData {keys [
      {arriveTangent=-0.0013, interpMode=RCIM_Cubic, leaveTangent=-0.0013, time=0}
      <at 15 0.98 RCIM_Cubic>
      {arriveTangent=0.0013, interpMode=RCIM_Cubic, leaveTangent=0.0013, time=30}
     ]}}
   volumeCurve {editorCurveData {keys [<at 0 1>]}}
  }
 }
 {
  id=Enveloper_1
  children [Enveloper_3]
  properties {pitchCurve {editorCurveData {keys [{time=0}]}}, volumeCurve {editorCurveData {keys [<at 0 0>,<at 1 1>]}}}
 }
 {
  id=Enveloper_3
  children [Enveloper_0]
  properties {
   bLoop=true
   bLoopIndefinitely=true
   loopEnd=20
   pitchCurve {editorCurveData {keys [{time=0}]}}
   volumeCurve {editorCurveData {keys [
      {arriveTangent=-0.01, interpMode=RCIM_Cubic, leaveTangent=-0.01, time=0, value=1}
      <at 10 0.9 RCIM_Cubic>
      {arriveTangent=0.01, interpMode=RCIM_Cubic, leaveTangent=0.01, time=20, value=1}
     ]}}
  }
 }
 {id=Mixer_0, children [Random_0,WavePlayer_0], properties {inputVolume [0.5,1]}}
 {id=Modulator_1, children [Mixer_0], properties {pitchMax=1, pitchMin=0.9}}
 {id=Random_0, children [WavePlayer_1,WavePlayer_2,WavePlayer_3], properties {weights [1,1,1]}}
 {
  id=WavePlayer_0
  properties {bLooping=true, soundWaveAssetPtr=/Game/Sounds/Artemis/Environment_Actors/ColiseumBrazierFire/Spot_Coliseum_Brazier_Fire_Stereo}
 }
 {id=WavePlayer_1, properties {bLooping=true, soundWaveAssetPtr=/Game/Sounds/Trans_AmbientLoops/Fire/Fire_Torch_01}}
 {id=WavePlayer_2, properties {bLooping=true, soundWaveAssetPtr=/Game/Sounds/Trans_AmbientLoops/Fire/Fire_Torch_02}}
 {id=WavePlayer_3, properties {bLooping=true, soundWaveAssetPtr=/Game/Sounds/Trans_AmbientLoops/Fire/Fire_Torch_03}}
]
properties {
 attenuationSettings=&/Spot_WallSconce_Fire_Attenuation
 bBypassVolumeScaleForPriority=true
 bOverrideConcurrency=true
 maxDistance=500
 priority=0
 soundClassObject=/Game/Sounds/FortSoundClassesAndModes/Fort_Ambient_NonLoop
 volumeMultiplier=0.6
 attenuationOverrides {attenuationShape=Cone, attenuationShapeExtents {x=171.3182}}
 concurrencyOverrides {maxCount=3, voiceStealReleaseTime=3}
}
